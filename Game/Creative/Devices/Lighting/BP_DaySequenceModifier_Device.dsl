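; Creative Day Sequence device that overrides time of day, sun, skylight, fog, clouds and sky, globally or within a blended volume.
; It waits for the day sequence actor, sets the volume's blend target to the player controller, then configures its modifier component from user options.
; Overrides are built into a transient procedural sequence unless a user sequence is supplied.
; Phase-based enabling, client-only timeline fades and legacy Skydome option migration are also handled.
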(event Custom|Setup%20TODM
  (bind _returnvalue (WorldSubsystems|GetDaySequenceSubsystem))
  (if (Replay|GetIsPlayingReplay)
    (Variables|Default|SetWaitForTimeOfDayManagerAsyncTask 0)
    (else
      _))
  (Utilities|FlowControl|Sequence
    (:then_0
      (Variables|Default|GetWaitForTimeOfDayManagerAsyncTask
        (:then
          _)
        (:else
          (bind (_asynctask _todm) (Default|GetTODM self)
            (:OnAsyncTaskActivated
              (Fortnite|FortShippingLog self (Utilities|String|Append "Set Async Task" (Utilities|GetDisplayName _asynctask)))
              (Variables|Default|SetWaitForTimeOfDayManagerAsyncTask _asynctask))
            (:Out
              (Fortnite|FortShippingLog self (Utilities|String|Append "TimeOfDayManager Updated: " (Utilities|GetDisplayName (Utilities|ToObject(Interface) _todm))))
              (Utilities|BindtoDaySequenceActor _todm
                (:Success
                  (Fortnite|FortShippingLog self (Utilities|String|Append "Cached TimeOfDayManager Updated: " (Utilities|GetDisplayName (Variables|Default|GetCached%20Fort%20Day%20Sequence%20Actor))))))))))
      (CallFunction|PostTODMCached))
    (:then_1
      ;[1 1 1] Listen for TODM changes while in editor (we no longer rely on async actions in editor, only in game)
      (if (Editor|InEditor)
        (DaySequence|UnbindEventfromOnDaySequenceActorSet _returnvalue (delegate Custom|OnDaySequenceActorSet))
        (DaySequence|AssignOnDaySequenceActorSet _returnvalue (delegate Custom|OnDaySequenceActorSet))))))

(event Custom|OnDaySequenceActorSet
  (params (NewActor :type Day%20Sequence%20Actor%20Object%20Reference))
  ;[1 1 1] Listen for TODM changes while in editor (we no longer rely on async actions in editor, only in game)
  (CallFunction|SetupTODM))

(event Custom|Initialize%20Component%20Values
  (bind _daysequencemodifier (Variables|Default|GetDaySequenceModifier))
  (if (or (Class|CreativeEnabledComponent|IsEnabled (Variables|Default|GetEnabledComponent)) (and (Editor|InEditororPIE) (not (Fortnite|IsCreativeorPlaygroundModePlaylist))))
    (DaySequence|EnableComponent _daysequencemodifier)
    (else
      (DaySequence|DisableComponent _daysequencemodifier)))
  (Variables|/Script/DaySequence.DaySequenceModifierComponent|SetBias (Variables|UserOptions|Basic|GetPriority) _daysequencemodifier)
  (Variables|/Script/DaySequence.DaySequenceModifierComponent|SetMode (select (Variables|UserOptions|Basic|Volume|GetUse%20Volume) (select (Utilities|Enum|Equal(Enum) (Variables|UserOptions|Basic|Volume|GetBlending) "Time") "Time" "Volume") (Utilities|Enum|LiteralenumEDaySequenceModifierMode "Global")) _daysequencemodifier)
  (Variables|/Script/DaySequence.DaySequenceModifierComponent|SetDayNightCycle (Variables|UserOptions|Basic|TimeOfDay|GetDayNightCycleMode) _daysequencemodifier)
  (Variables|/Script/DaySequence.DaySequenceModifierComponent|SetDayNightCycleTime (Variables|UserOptions|Basic|TimeOfDay|GetDayNightCycleTime) _daysequencemodifier)
  (Variables|/Script/DaySequence.DaySequenceModifierComponent|SetBlendAmount (select (Variables|UserOptions|Basic|Volume|GetUse%20Volume) (Utilities|Select :Index (Variables|UserOptions|Basic|Volume|GetBlending) :Global 0.0 :Volume (Variables|UserOptions|Basic|Volume|GetBlendAmount) :Time (Variables|UserOptions|Basic|Volume|GetBlendAmount)) 0.0) _daysequencemodifier)
  (Variables|/Script/DaySequence.DaySequenceModifierComponent|SetBlendTime (select (Variables|UserOptions|Basic|Volume|GetUse%20Volume) (Utilities|Select :Index (Variables|UserOptions|Basic|Volume|GetBlending) :Global 0.0 :Volume (Variables|UserOptions|Basic|Volume|GetBlendTime) :Time (Variables|UserOptions|Basic|Volume|GetBlendTime)) 0.0) _daysequencemodifier)
  (Variables|/Script/DaySequence.DaySequenceModifierComponent|SetbPreview (and (Class|CreativeEnabledComponent|IsEnabled (Variables|Default|GetEnabledComponent)) (Variables|Preview|GetPreview%20In%20Editor)) _daysequencemodifier)
  (Rendering|SetVisibility _daysequencemodifier false false)
  (Components|Timeline|SetPlayRate (Variables|BP_DaySequenceModifier_Device|GetCustomVolumeBlendTimeline) (Variables|UserOptions|Basic|GetPlayRate)))

(event EventBeginPlay
  (|CallParentFunction)
  (CallFunction|SetupTODM))

(event Custom|Setup%20Volume
  ;[1 1 1] Setup Volume
  (CallFunction|SetVolumeSize)
  (Rendering|SetVisibility (Variables|Default|GetVolumeVisualizer) (and (Class|CreativeEnabledComponent|IsEnabled (Variables|Default|GetEnabledComponent)) (Variables|UserOptions|Basic|Volume|GetUse%20Volume)) false))

(event EventInitializeDevice
  (|CallParentFunction)
  (CallFunction|SetLabelOverridefromSkydome)
  (Class|CreativeEnabledComponent|Initialize (Variables|Default|GetEnabledComponent) (Variables|Default|GetFortMinigameProgress) (Variables|UserOptions|Basic|GetEnableDuringPhase))
  (CallFunction|SetEnabledDuringPhase)
  (CallFunction|SetupTODM))

(event Custom|Enable%20Event
  (params (Instigator :type Controller%20Object%20Reference))
  ;[0.06965 0 1] Enable / Disable Receivers
  (Networking|SwitchHasAuthority
    (:Authority
      (Class|CreativeEnabledComponent|SetEnabled (Variables|Default|GetEnabledComponent) true))))

(event OnAddedtoMinigame(FortMinigameProgress)
  (params (Minigame :type Fort%20Minigame%20Object%20Reference))
  (Variables|Default|SetCachedMinigameRef Minigame)
  (CallFunction|SetEnabledDuringPhase)
  (CallFunction|SetupTODM))

(event Custom|Disable%20Event
  (params (Instigator :type Controller%20Object%20Reference))
  ;[0.06965 0 1] Enable / Disable Receivers
  (Networking|SwitchHasAuthority
    (:Authority
      (Class|CreativeEnabledComponent|SetEnabled (Variables|Default|GetEnabledComponent) false))))

(event Custom|Post%20TODM%20Cached
  (CallFunction|SetupControllerEvents) ; Will call On Blend Target Set when the player controller is ready
)

(event Custom|On%20Blend%20Target%20Set
  (CallFunction|SetupDevice))

(event OnEnabledStateChanged(EnabledComponent)
  (params (Enabled :type Boolean))
  ;[0.1253 1 0.1005] EnabledComponent Events
  (CallFunction|SetupDevice))

(event Custom|RemoveDevice
  (bind _waitfortimeofdaymanagerasynctask (Variables|Default|GetWaitForTimeOfDayManagerAsyncTask)
    (:then
      (Class|FortAsyncActionWaitforTimeOfDayManagerUpdated|EndTask _waitfortimeofdaymanagerasynctask)
      (Variables|Default|SetWaitForTimeOfDayManagerAsyncTask 0))
    (:else)))

(event Custom|FadeIn%20Event
  (params (Instigator :type Controller%20Object%20Reference))
  ;[0 1 0.9051] FadeIn/FadeOut Events
  (Networking|SwitchHasAuthority
    (:Remote
      (CallFunction|SetTimelineInterpolatedBlendTime (Variables|BP_DaySequenceModifier_Device|GetCustomVolumeBlendTimeline))
      (bind (_direction _smoothfade) (Timeline|CustomVolumeBlendTimeline Play)
        (:Update
          (Variables|/Script/DaySequence.DaySequenceModifierComponent|SetUserBlendWeight _smoothfade (Variables|Default|GetDaySequenceModifier)))))) ; These events are cosmetic and are meaningless on the server.
)

(event Custom|FadeOut%20Event
  (params (Instigator :type Controller%20Object%20Reference))
  ;[0 1 0.9051] FadeIn/FadeOut Events
  (Networking|SwitchHasAuthority
    (:Remote
      (CallFunction|SetTimelineInterpolatedBlendTime (Variables|BP_DaySequenceModifier_Device|GetCustomVolumeBlendTimeline))
      (Timeline|CustomVolumeBlendTimeline Reverse))))

(event EventEndPlay
  (params (EndPlayReason :type EEndPlayReason%20Enum :default "Destroyed"))
  (CallFunction|RemoveDevice)
  (|CallParentFunction EndPlayReason))

(event Custom|Setup%20Controller%20Events
  ;[1 1 1] PlayerController/ViewTarget Events
  ;;[1 1 1] Get PlayerController, bind to OnViewTargetChanged, and get initial view target.
  (CallFunction|ClearPlayerControllerSetup)
  (if (and (Variables|UserOptions|Basic|Volume|GetUse%20Volume) (not (Editor|InEditor)))
    (bind _returnvalue (Utilities|PlayerControllerWaitLoop)
      (:Is%20Valid
        (bind _asfort_replay_spectator (Utilities|Casting|CastToFortReplaySpectator _returnvalue)
          (:then
            (EventDispatchers|BindEventtoOnTimelineScrubbed _asfort_replay_spectator (EventDispatchers|CreateEvent :!event Custom|Setup%20TODM)))
          (:CastFailed
            _))
        (CallFunction|EventSetBlendTarget _returnvalue)))
    (else
      (CallFunction|OnBlendTargetSet) ; Early out for in editor or global device
    )))

(event Custom|Event%20SetBlendTarget
  (params (PlayerController :type Fort%20Player%20Controller%20Object%20Reference))
  (bind _daysequencemodifier (Variables|Default|GetDaySequenceModifier))
  ;[1 1 1] PlayerController/ViewTarget Events
  ;;[1 1 1] Set Blend Target to PC
  (DaySequence|SetBlendTarget _daysequencemodifier PlayerController)
  (Variables|/Script/DaySequence.DaySequenceModifierComponent|SetUserBlendWeight 1.0 _daysequencemodifier)
  (CallFunction|OnBlendTargetSet))

(event Custom|Clear%20Player%20Controller%20Setup
  ;[1 1 1] PlayerController/ViewTarget Events
  ;;[1 1 1] Clean up any previous PC setup
  (DaySequence|SetBlendTarget (Variables|Default|GetDaySequenceModifier) 0)
  (bind _cached_fort_player_controller (Variables|Default|GetCached%20Fort%20Player%20Controller)
    (:then
      (bind _asfort_replay_spectator (Utilities|Casting|CastToFortReplaySpectator _cached_fort_player_controller)
        (:then
          (EventDispatchers|UnbindEventfromOnTimelineScrubbed _asfort_replay_spectator (EventDispatchers|CreateEvent :!event Custom|Setup%20TODM)))
        (:CastFailed
          _)))
    (:else
      _))
  (Variables|Default|SetCached%20Fort%20Player%20Controller 0))

(fn AddCloudOverrides
  (params (InBuilder :type Procedural%20Day%20Sequence%20Builder%20Object%20Reference))
  (LocalVariables|SetLocalBuilder InBuilder)
  (if (Variables|UserOptions|Basic|Clouds|GetOverrideClouds)
    (Utilities|FlowControl|Sequence
      (:then_0
        (DaySequence|SetActiveBoundObject (LocalVariables|GetLocalBuilder) (Variables|/Script/FortniteGame.FortDaySequenceActor|GetSkyDomeComponent (Variables|Default|GetCached%20Fort%20Day%20Sequence%20Actor)))
        (DaySequence|AddScalarMaterialParameterOverride (LocalVariables|GetLocalBuilder) "SkyboxClouds" 0 1.0)
        (DaySequence|AddScalarMaterialParameterOverride (LocalVariables|GetLocalBuilder) "CloudLightStrength" 0 (Variables|UserOptions|Basic|Clouds|GetCloudLightingBrightness))
        (DaySequence|AddScalarMaterialParameterOverride (LocalVariables|GetLocalBuilder) "CloudCoverage" 0 (Variables|UserOptions|Basic|Clouds|GetCloudCoverage))
        (DaySequence|AddScalarMaterialParameterOverride (LocalVariables|GetLocalBuilder) "CloudOpacity" 0 (Variables|UserOptions|Basic|Clouds|GetCloudOpacity))
        (DaySequence|AddColorMaterialParameterOverride (LocalVariables|GetLocalBuilder) "CloudColor_Custom" 0 (Variables|UserOptions|Basic|Clouds|GetCloudLightColor))
        (DaySequence|AddColorMaterialParameterOverride (LocalVariables|GetLocalBuilder) "CloudShadowColor_Custom" 0 (Variables|UserOptions|Basic|Clouds|GetCloudShadowColor))
        (DaySequence|AddScalarMaterialParameterOverride (LocalVariables|GetLocalBuilder) "CloudSize" 0 (Variables|UserOptions|Basic|Clouds|GetCloudSize))
        (DaySequence|AddScalarMaterialParameterOverride (LocalVariables|GetLocalBuilder) "UseCustomCloudColor" 0 1.0)
        (DaySequence|AddColorMaterialParameterOverride (LocalVariables|GetLocalBuilder) "CloudDirection" 0 (Math|Conversions|ToLinearColor(Vector) (Math|Vector|MakeVector (Variables|UserOptions|Basic|Clouds|GetCloudDirectionX) (Variables|UserOptions|Basic|Clouds|GetCloudDirectionY))))
        (DaySequence|AddScalarMaterialParameterOverride (LocalVariables|GetLocalBuilder) "CloudSpeed" 0 (Variables|UserOptions|Basic|Clouds|GetCloudSpeed)))
      (:then_1
        (DaySequence|SetActiveBoundObject (LocalVariables|GetLocalBuilder) (Variables|/Script/FortniteGame.FortDaySequenceActor|GetVolumetricCloudComponent (Variables|Default|GetCached%20Fort%20Day%20Sequence%20Actor)))
        (DaySequence|AddVisibilityOverride (LocalVariables|GetLocalBuilder) false)))))

(fn AddFogOverrides
  (params (InBuilder :type Procedural%20Day%20Sequence%20Builder%20Object%20Reference))
  (bind _fogcolor (Variables|UserOptions|Basic|Fog|GetFogColor))
  (LocalVariables|SetLocalBuilder InBuilder)
  (if (Variables|UserOptions|Basic|Fog|GetOverrideFog)
    (Utilities|FlowControl|Sequence
      (:then_0
        (DaySequence|SetActiveBoundObject (LocalVariables|GetLocalBuilder) (TimeOfDay|GetDirectionalLightComponent (Variables|Default|GetCached%20Fort%20Day%20Sequence%20Actor))) ; volumetric scattering
        (DaySequence|AddScalarOverride (LocalVariables|GetLocalBuilder) "VolumetricScatteringIntensity" (Variables|UserOptions|Basic|Fog|GetSunlightVolumetricScattering)))
      (:then_1
        (DaySequence|SetActiveBoundObject (LocalVariables|GetLocalBuilder) (Variables|/Script/FortniteGame.FortDaySequenceActor|GetExponentialHeightFogComponent (Variables|Default|GetCached%20Fort%20Day%20Sequence%20Actor)))
        (DaySequence|AddVisibilityOverride (LocalVariables|GetLocalBuilder) (Variables|UserOptions|Basic|Fog|GetEnableFogComponent))
        (DaySequence|AddColorOverride (LocalVariables|GetLocalBuilder) "VolumetricFogAlbedo" _fogcolor)
        (DaySequence|AddColorOverride (LocalVariables|GetLocalBuilder) "FogInscatteringLuminance" _fogcolor)
        (DaySequence|AddScalarOverride (LocalVariables|GetLocalBuilder) "FogDensity" (Variables|UserOptions|Basic|Fog|GetFogDensity))
        (DaySequence|AddScalarOverride (LocalVariables|GetLocalBuilder) "FogHeightFalloff" (Variables|UserOptions|Basic|Fog|GetFogHeightFalloff))
        (DaySequence|AddColorOverride (LocalVariables|GetLocalBuilder) "DirectionalInscatteringLuminance" (* (Variables|UserOptions|Basic|Fog|GetFogDirectionalInscatteringColor) (Variables|UserOptions|Basic|Fog|GetSunlightVolumetricScattering)))
        (DaySequence|AddScalarOverride (LocalVariables|GetLocalBuilder) "FogMaxOpacity" (Variables|UserOptions|Basic|Fog|GetFogMaxOpacity))
        (DaySequence|AddScalarOverride (LocalVariables|GetLocalBuilder) "SecondFogData.FogDensity" (Variables|UserOptions|Basic|Fog|GetSecondaryFogDensity))
        (DaySequence|AddScalarOverride (LocalVariables|GetLocalBuilder) "SecondFogData.FogHeightFalloff" (Variables|UserOptions|Basic|Fog|GetSecondaryFogFalloff))
        (DaySequence|AddScalarOverride (LocalVariables|GetLocalBuilder) "SecondFogData.FogHeightOffset" (Variables|UserOptions|Basic|Fog|GetSecondaryFogOffset))))))

(fn AddLightOverrides
  (params (InBuilder :type Procedural%20Day%20Sequence%20Builder%20Object%20Reference))
  (LocalVariables|SetLocalBuilder InBuilder)
  (if (Variables|UserOptions|Basic|Sunlight|GetOverrideSunlight)
    (Utilities|FlowControl|Sequence
      (:then_0
        (DaySequence|SetActiveBoundObject (LocalVariables|GetLocalBuilder) (TimeOfDay|GetDirectionalLightComponent (Variables|Default|GetCached%20Fort%20Day%20Sequence%20Actor)))
        (DaySequence|AddVisibilityOverride (LocalVariables|GetLocalBuilder) (Variables|UserOptions|Basic|Sunlight|GetEnableSunComponent))
        (DaySequence|AddScalarOverride (LocalVariables|GetLocalBuilder) "Intensity" (Variables|UserOptions|Basic|Sunlight|GetSunlightIntensity))
        (DaySequence|AddColorOverride (LocalVariables|GetLocalBuilder) "LightColor" (Variables|UserOptions|Basic|Sunlight|GetSunlightColor)))
      (:then_1
        (DaySequence|SetActiveBoundObject (LocalVariables|GetLocalBuilder) (Variables|/Script/FortniteGame.FortDaySequenceActor|GetSunFlareComponent (Variables|Default|GetCached%20Fort%20Day%20Sequence%20Actor)))
        (DaySequence|AddVisibilityOverride (LocalVariables|GetLocalBuilder) (Variables|UserOptions|Basic|Sunlight|GetEnableSunFlare)))
      (:then_2
        (DaySequence|SetActiveBoundObject (LocalVariables|GetLocalBuilder) (Variables|/Script/FortniteGame.FortDaySequenceActor|GetSunRootComponent (Variables|Default|GetCached%20Fort%20Day%20Sequence%20Actor)))
        (DaySequence|AddTransformOverride (LocalVariables|GetLocalBuilder) (Math|Transform|MakeTransform "100.000000,0.000000,0.000000" (Math|Rotator|MakeRotator (Variables|UserOptions|Basic|Sunlight|GetSunlightRotationX) 0.0 (Variables|UserOptions|Basic|Sunlight|GetSunlightRotationZ))))))))

(fn AddSkylightOverrides
  (params (InBuilder :type Procedural%20Day%20Sequence%20Builder%20Object%20Reference))
  (LocalVariables|SetLocalBuilder InBuilder)
  (if (Variables|UserOptions|Basic|Skylight|GetOverrideSkylight)
    (DaySequence|SetActiveBoundObject (LocalVariables|GetLocalBuilder) (Variables|/Script/FortniteGame.FortDaySequenceActor|GetSkyLightComponent (Variables|Default|GetCached%20Fort%20Day%20Sequence%20Actor)))
    (DaySequence|AddVisibilityOverride (LocalVariables|GetLocalBuilder) (Variables|UserOptions|Basic|Skylight|GetEnableSkylightComponent))
    (DaySequence|AddScalarOverride (LocalVariables|GetLocalBuilder) "Intensity" (Variables|UserOptions|Basic|Skylight|GetSkylightIntensity))
    (DaySequence|AddColorOverride (LocalVariables|GetLocalBuilder) "LightColor" (Variables|UserOptions|Basic|Skylight|GetSkylightColor))))

(fn AddSkyOverrides
  (params (InBuilder :type Procedural%20Day%20Sequence%20Builder%20Object%20Reference))
  (LocalVariables|SetLocalBuilder InBuilder)
  (if (Variables|UserOptions|Basic|Sky|GetOverrideSky)
    (Utilities|FlowControl|Sequence
      (:then_0
        (DaySequence|SetActiveBoundObject (LocalVariables|GetLocalBuilder) (Variables|/Script/FortniteGame.FortDaySequenceActor|GetSkyAtmosphereComponent (Variables|Default|GetCached%20Fort%20Day%20Sequence%20Actor)))
        (DaySequence|AddVisibilityOverride (LocalVariables|GetLocalBuilder) (Variables|UserOptions|Basic|Sky|GetEnableSkyAtmosphereComponent)))
      (:then_1
        (DaySequence|SetActiveBoundObject (LocalVariables|GetLocalBuilder) (Variables|/Script/FortniteGame.FortDaySequenceActor|GetSkyDomeComponent (Variables|Default|GetCached%20Fort%20Day%20Sequence%20Actor)))
        (DaySequence|AddVisibilityOverride (LocalVariables|GetLocalBuilder) (Variables|UserOptions|Basic|Sky|GetEnableSkyboxComponent))
        (Utilities|FlowControl|Sequence
          (:then_0
            ;[1 1 1] Sky Gradient Overrides
            (DaySequence|AddScalarMaterialParameterOverride (LocalVariables|GetLocalBuilder) "GradientColorStrength" 0 (Variables|UserOptions|Basic|Sky|GetSkyGradientBlend))
            (DaySequence|AddColorMaterialParameterOverride (LocalVariables|GetLocalBuilder) "ColorLow" 0 (Variables|UserOptions|Basic|Sky|GetSkyGradientLowColor))
            (DaySequence|AddColorMaterialParameterOverride (LocalVariables|GetLocalBuilder) "ColorMid" 0 (Variables|UserOptions|Basic|Sky|GetSkyGradientMidColor))
            (DaySequence|AddColorMaterialParameterOverride (LocalVariables|GetLocalBuilder) "ColorHigh" 0 (Variables|UserOptions|Basic|Sky|GetSkyGradientHighColor)))
          (:then_1
            ;[1 1 1] Sun Overrides
            (DaySequence|AddScalarMaterialParameterOverride (LocalVariables|GetLocalBuilder) "SunSize" 0 (Variables|UserOptions|Basic|Sky|GetSunSize))
            (DaySequence|AddScalarMaterialParameterOverride (LocalVariables|GetLocalBuilder) "SunBrightness" 0 (Math|Float|Power (Variables|UserOptions|Basic|Sky|GetSunIntensity) 4.0))
            (DaySequence|AddColorMaterialParameterOverride (LocalVariables|GetLocalBuilder) "SunColor" 0 (Variables|UserOptions|Basic|Sky|GetSunColor))
            (DaySequence|AddScalarMaterialParameterOverride (LocalVariables|GetLocalBuilder) "SunColorBlend" 0 (Variables|UserOptions|Basic|Sky|GetCustomSunColorBlend)))
          (:then_2
            ;[1 1 1] Moon Overrides
            (DaySequence|AddScalarMaterialParameterOverride (LocalVariables|GetLocalBuilder) "MoonSize" 0 (Utilities|RemapFloat (Variables|UserOptions|Basic|Sky|GetMoonSize) "(X=0.000000,Y=1.000000)" "(X=0.010000,Y=0.330000)" true))
            (DaySequence|AddColorMaterialParameterOverride (LocalVariables|GetLocalBuilder) "MoonColor" 0 (Variables|UserOptions|Basic|Sky|GetMoonColor))
            (DaySequence|AddScalarMaterialParameterOverride (LocalVariables|GetLocalBuilder) "MoonIntensity" 0 (Variables|UserOptions|Basic|Sky|GetMoonIntensity))
            (DaySequence|AddScalarMaterialParameterOverride (LocalVariables|GetLocalBuilder) "MoonHaloIntensity" 0 (Variables|UserOptions|Basic|Sky|GetMoonHaloIntensity))
            (DaySequence|AddScalarMaterialParameterOverride (LocalVariables|GetLocalBuilder) "MoonHaloSize" 0 (Variables|UserOptions|Basic|Sky|GetMoonHaloSize))
            (DaySequence|AddScalarMaterialParameterOverride (LocalVariables|GetLocalBuilder) "StarsBrightness" 0 (Variables|UserOptions|Basic|Sky|GetStarBrightness))))))))

(fn CheckCameraInVolume
  (if (Variables|UserOptions|Basic|Volume|GetUse%20Volume)
    _))

(fn ConstructSequence
  (bind _localbuilder (LocalVariables|GetLocalBuilder))
  (Variables|Default|GetCached%20Fort%20Day%20Sequence%20Actor
    (:then
      (bind _userdaysequence (Variables|UserOptions|Basic|GetUserDaySequence)
        (:then
          (Variables|/Script/DaySequence.DaySequenceModifierComponent|SetUserDaySequence _userdaysequence (Variables|Default|GetDaySequenceModifier)))
        (:else
          ;[1 1 1] Prepare the transient Procedural Sequence
          (bind _returnvalue (Game|ConstructObjectfromClass "/Script/DaySequence.ProceduralDaySequenceBuilder"))
          (bind _output_get (LocalVariables|SetLocalBuilder _returnvalue))
          (bind _returnvalue_2 (DaySequence|Initialize _output_get (Variables|Default|GetCached%20Fort%20Day%20Sequence%20Actor) 0 true))
          (Variables|Default|SetProceduralSequence _returnvalue_2)
          (Utilities|FlowControl|Sequence
            (:then_0
              ;[1 1 1] Add Override Tracks
              (Utilities|FlowControl|Sequence
                (:then_0
                  (ProceduralSequence|AddLightOverrides _localbuilder))
                (:then_1
                  (ProceduralSequence|AddFogOverrides _localbuilder))
                (:then_2
                  (ProceduralSequence|AddSkylightOverrides _localbuilder))
                (:then_3
                  (ProceduralSequence|AddCloudOverrides _localbuilder))
                (:then_4
                  (ProceduralSequence|AddSkyOverrides _localbuilder))))
            (:then_1
              ;[1 1 1] Apply the Procedural Sequence to the Modifier Component
              (Variables|/Script/DaySequence.DaySequenceModifierComponent|SetTransientSequence (Variables|Default|GetProceduralSequence) (Variables|Default|GetDaySequenceModifier)))) ; build, then set
        )))
    (:else)))

(fn GetCollisionStaticMesh
  (return "0, 0, 0" (Variables|/Script/Engine.StaticMeshComponent|GetStaticMesh (Variables|Default|GetEditOnly_Icon))))

(fn GetMeshComponents
  (return (Utilities|Array|MakeArray (Variables|Default|GetEditOnly_Icon))))

(fn IgnoreWorldGridScaleOverride
  (return false))

(fn MigrateLegacyUserOptions
  (bind (_use_volume _blendtime _skydomevolumeheight) (Class|GetClassDefaults "/Game/Creative/Devices/Lighting/BP_DaySequenceModifier_Device.BP_DaySequenceModifier_Device_C"))
  (bind (_use_volume_2 _blendtime_2 _skydomevolumedepth) (Class|GetClassDefaults "/Game/Creative/Devices/Lighting/BP_DaySequenceModifier_Device.BP_DaySequenceModifier_Device_C"))
  (bind (_use_volume_3 _blendtime_3 _enableduringphase) (Class|GetClassDefaults "/Game/Creative/Devices/Lighting/BP_DaySequenceModifier_Device.BP_DaySequenceModifier_Device_C"))
  (bind (_use_volume_4 _blendtime_4 _skydomevolumewidth) (Class|GetClassDefaults "/Game/Creative/Devices/Lighting/BP_DaySequenceModifier_Device.BP_DaySequenceModifier_Device_C"))
  (bind (_use_volume_5 _blendtime_5 _skydomeenabledduringphase) (Class|GetClassDefaults "/Game/Creative/Devices/Lighting/BP_DaySequenceModifier_Device.BP_DaySequenceModifier_Device_C"))
  (bind _skydomevolumewidth_2 (Variables|Migration|GetSkydomeVolumeWidth))
  (bind _skydomevolumedepth_2 (Variables|Migration|GetSkydomeVolumeDepth))
  (bind _skydomevolumeheight_2 (Variables|Migration|GetSkydomeVolumeHeight))
  (|CallParentFunction)
  (Utilities|FlowControl|Sequence
    (:then_0
      (Utilities|FlowControl|Sequence
        (:then_0
          ;[1 1 1] Handle conversion of skydome device settings to day sequence device for FNC islands
          (if (!= _skydomevolumewidth_2 _skydomevolumewidth)
            (Variables|UserOptions|Basic|Volume|SetVolumeExtentX (* _skydomevolumewidth_2 5.0))
            (Variables|Migration|SetSkydomeVolumeWidth _skydomevolumewidth)
            (LocalVariables|SetbOptionsMigrated true)))
        (:then_1
          (if (!= _skydomevolumedepth_2 _skydomevolumedepth)
            (Variables|UserOptions|Basic|Volume|SetVolumeExtentY (* _skydomevolumedepth_2 5.0))
            (Variables|Migration|SetSkydomeVolumeDepth _skydomevolumedepth)
            (LocalVariables|SetbOptionsMigrated true)))
        (:then_2
          (if (!= _skydomevolumeheight_2 _skydomevolumeheight)
            (Variables|UserOptions|Basic|Volume|SetVolumeExtentZ (* _skydomevolumeheight_2 5.0))
            (Variables|Migration|SetSkydomeVolumeHeight _skydomevolumeheight)
            (LocalVariables|SetbOptionsMigrated true)))
        (:then_3
          (if (Utilities|Enum|NotEqual(Enum) (Variables|Migration|GetSkydomeEnabledDuringPhase) _skydomeenabledduringphase)
            (Variables|Migration|SetSkydomeEnabledDuringPhase _skydomeenabledduringphase)
            (bind _migratedproperties (DeviceConversion|MigrateLegacyEnabledIndex (Variables|UserOptions|Basic|GetEnableDuringPhase) _skydomeenabledduringphase))
            (LocalVariables|SetbOptionsMigrated _migratedproperties)))
        (:then_4
          ;[1 1 1] Set internal legacy Enabled Index
          (bind _migratedproperties_2 (DeviceConversion|MigrateLegacyEnabledIndex (Variables|UserOptions|Basic|GetEnableDuringPhase) _enableduringphase))
          (LocalVariables|SetbOptionsMigrated _migratedproperties_2))))
    (:then_1
      (if (LocalVariables|GetbOptionsMigrated)
        (OptionData|RefreshPropertyOverrides (Variables|Default|GetToyOptionsComponent))))))

(fn SetEnabledDuringPhase
  (bind _enabledcomponent (Variables|Default|GetEnabledComponent))
  (bind _enabled (Class|CreativeEnabledComponent|IsEnabled _enabledcomponent))
  (Utilities|FlowControl|Sequence
    (:then_0
      (LocalVariables|SetCached%20Enabled%20State _enabled))
    (:then_1
      (Class|CreativeEnabledComponent|SetEnabledOnPhase (Variables|Default|GetEnabledComponent) (select (Utilities|IsValid (Variables|Default|GetCachedMinigameRef) :!class /Script/Engine.KismetSystemLibrary) (Variables|/Script/FortniteGame.FortMinigame|GetCurrentState (Variables|Default|GetCachedMinigameRef)) "PreGame") (DeviceConversion|EnabledPhaseEnumtoLegacyInt (Variables|UserOptions|Basic|GetEnableDuringPhase)))
      (if (or (== _enabled (LocalVariables|GetCached%20Enabled%20State)) (Editor|InEditor))
        ;[1 1 1] When the state didn't change, we still want to update settings based on if it's enabled or not.
        (CallFunction|SetupDevice)))))

(fn SetLabelOverrideFromSkydome
  (if (Variables|Migration|GetbMigrateLabelOverride)
    (LabelOverride|SetLabelOverride 0 (Utilities|String|Replace (LabelOverride|GetLabelOverride) "Skydome" "Day Sequence Device")) ; Calling from server, player controller is null
    (Variables|Migration|SetbMigrateLabelOverride false)))

(fn SetTimelineInterpolatedBlendTime
  (params (Timeline :type Timeline%20Component%20Object%20Reference))
  ;[1 1 1] Set timeline position based on current blend weight
  (Components|Timeline|SetNewTime Timeline (Math|Float|NormalizetoRange (DaySequence|GetBlendWeight (Variables|Default|GetDaySequenceModifier)) 0.0 (Components|Timeline|GetTimelineLength Timeline))))

(fn SetupDevice
  (Utilities|FlowControl|Sequence
    (:then_0
      (CallFunction|InitializeComponentValues))
    (:then_1
      (CallFunction|SetupVolume))
    (:then_2
      (ProceduralSequence|ConstructSequence))))

(fn SetVolumeSize
  (Transformation|SetRelativeScale3D (Variables|Default|GetVolumeVisualizer) (Math|Vector|MakeVector (Math|Float|Max(Float) (Variables|UserOptions|Basic|Volume|GetVolumeExtentX) 0.05) (Math|Float|Max(Float) (Variables|UserOptions|Basic|Volume|GetVolumeExtentY) 0.05) (Math|Float|Max(Float) (Variables|UserOptions|Basic|Volume|GetVolumeExtentZ) 0.05)))
  (Transformation|SetWorldRotation (Variables|Default|GetTriggerParent) "0, 0, 0" false false))

(fn ConstructionScript
  (|CallParentFunction))
